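# Material function for the RoseDust character that optionally overlays distance-field inklines on incoming material attributes, writing line color into BaseColor and Emissive.
# A sphere mask and power-based blur shape the DF texture into lines, with thickness varying by camera distance.
# Color and opacity come from a parameter or texture, with optional LUT on color.
# Static switches allow separate DF textures for low quality or mobile.
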MaterialFunction /Game/Characters/Player/Female/Medium/Bodies/F_MED_RoseDust/Materials/MF_RoseDust_DF_Inklines
graph {nodes [
  n1 = NamedRerouteUsage()
  n2 = GetMaterialAttributes(None=@n1) {attributeGetTypes [#1]}
  n3 = LinearInterpolate(A=@n2.BaseColor, B=NamedRerouteUsage(), Alpha=Saturate(None=NamedRerouteUsage()))
  n4 = SetMaterialAttributes!(MaterialAttributes=@n2.MaterialAttributes, BaseColor=@n3, EmissiveColor=@n3) {attributeSetTypes [#1, #2]}
  n5 = NamedRerouteUsage()
  FunctionOutput(None=StaticSwitchParameter(True=BlendMaterialAttributes(A=@n4, B=@n1, Alpha=@n5), False=NamedRerouteUsage()) {group=DF_Inklines, parameterName=UseDFInlines, sortPriority=0})
  NamedRerouteDeclaration(None=FunctionInput() {inputName=DFMap, inputType=FunctionInput_MaterialAttributes}) {name=MA, nodeColor=<rgba 1 0.5625 0 1>}
  n6 = VectorParameter() {group=DF_Inklines, parameterName=InkLine_Color, sortPriority=15, defaultValue=<rgba 0 0 0 1>}
  n7 = StaticSwitchParameter(
   True=TextureSampleParameter2D() {group=Toon_Shading, parameterName=DFC_Map, sortPriority=11, texture=&T_BlacksRGB}.RGBA
   False=AppendVector(A=@n6.RGB, B=@n6.RGB)
  ) {group=DF_Inklines, parameterName=UseInkLineColorTexture, sortPriority=1}
  NamedRerouteDeclaration(None=ComponentMask(None=@n7) {a=true}) {name=InkLineOpacity, nodeColor=<rgba 1 0.7734 0 1>}
  n8 = TextureSampleParameter2D() {
   group=DF_Inklines
   parameterName=DF_InkLines
   samplerSource=SSM_Wrap_WorldGroupSettings
   samplerType=SAMPLERTYPE_Alpha
   sortPriority=10
   texture=/Game/Global/Textures/Default/Blanks/T_BlackAlpha
  }
  n9 = TextureSampleParameter2D() {
   group=DF_Inklines
   parameterName=DF_InkLines_Mobile
   samplerSource=SSM_Wrap_WorldGroupSettings
   samplerType=SAMPLERTYPE_Alpha
   sortPriority=10
   texture=&T_BlackAlpha
  }
  n10 = StaticSwitchParameter(
   True=QualitySwitch(Default=@n8.R, Low=@n9.R, High=@n8.R, Medium=@n9.R, Epic=@n8.R)
   False=StaticSwitchParameter(True=ShadingPathSwitch(Default=@n8.R, Deferred=@n8.R, Mobile=@n9.R) {desc=Migrated from FeatureLevelSwitch}, False=@n8.R) {group=DF_Inklines, parameterName=Enable Unique Mobile DF Inklines, sortPriority=1}
  ) {group=DF_Inklines, parameterName=Enable Unique Low Quality DF Inklines, sortPriority=1}
  n11 = ScalarParameter() {defaultValue=2, group=DF_Inklines, parameterName=DF_BlurPow, sliderMax=16, sortPriority=23}
  n12 = OneMinus(None=Multiply(
    A=Power(Base=@n10, Exp=@n11) {constExponent=3}
    B=ScalarParameter() {defaultValue=1, group=DF_Inklines, parameterName=DF_BlurMult, sliderMax=2, sortPriority=24}
   ) {constB=0.5})
  n13 = Constant() {r=1}
  n14 = ScalarParameter() {defaultValue=100, group=DF_Inklines, parameterName=DF_FarDistance, sliderMax=1000, sortPriority=26}
  n15 = Subtract(A=Distance(A=WorldPosition().XYZ, B=CameraPositionWS()), B=@n14)
  n16 = Subtract(A=ScalarParameter() {defaultValue=500, group=DF_Inklines, parameterName=DF_NearDistance, sliderMax=1000, sortPriority=25}, B=@n14)
  n17 = LinearInterpolate(
   A=ScalarParameter() {defaultValue=0.4, group=DF_Inklines, parameterName=DF_LineThickness_Near, sliderMax=1, sortPriority=20}
   B=ScalarParameter() {defaultValue=0.6, group=DF_Inklines, parameterName=DF_LineThickness_Far, sliderMax=1, sortPriority=21}
   Alpha=Saturate(None=Divide(A=@n15, B=@n16))
  )
  n18 = ScalarParameter() {defaultValue=0.8, group=DF_Inklines, parameterName=DF_LineHardness, sliderMax=1, sortPriority=22}
  n19 = OneMinus(None=Saturate(None=SphereMask(A=@n10, B=@n13, Radius=@n17, Hardness=@n18) {attenuationRadius=0.1}))
  NamedRerouteDeclaration(None=Saturate(None=Multiply(A=@n12, B=@n19))) {name=InkLine Mask, nodeColor=<rgba 0.1615 0.1615 0.1615 1>}
  n20 = ComponentMask(None=@n7) {b=true, g=true, r=true}
  NamedRerouteDeclaration(None=StaticSwitchParameter(
    True=MaterialFunctionCall!(In=@n20, LUT, LUT Multiplier) {materialFunction=/Game/Packages/Fortress_SharedMaterials/Anime_Master_Material/MF_ApplyLUT}.Result
    False=@n20
   ) {group=DF_Inklines, parameterName=Use Apply LUT, sortPriority=1}) {name=InkLine Color, nodeColor=<rgba 0.1615 0.1615 0.1615 1>}
 ]}
