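# Material function for the RoseDust character that adds switchable comic-style rim lights, hatching, halftone dots and colored transition stripes to incoming material attributes.
# Screen-projected light vectors dotted with normals are thresholded against procedural patterns on a virtual plane anchored to the actor, scaled by FOV and distance.
# FX channel masks and vertex color mask the effects.
# Static switches select suit or skin/hair shading, with emissive passed through a LUT and metallic optionally reused as sheen.
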MaterialFunction /Game/Characters/Player/Female/Medium/Bodies/F_MED_RoseDust/Materials/MF_RoseDust_Highlights
graph {nodes [
  n1 = ScreenPosition()
  Ceil(None=Add(A=ComponentMask(None=@n1.ViewportUV) {r=true}) {constB=-0.5})
  CollectionParameter() {
   collection=&FortniteMaterialParameters
   parameterName=SunlightDirectionalVector
  }
  Constant() {r=1.4}
  Constant() {r=1.75}
  FunctionOutput(None=StaticSwitchParameter(True=NamedRerouteUsage(), False=NamedRerouteUsage()) {group=1_RoseDust, parameterName=UseRoseDustHighlights, sortPriority=0})
  n2 = Custom() {code="return View.DirectionalLightDirection;", description=DirectionalLightDirection, inputs []}
  n3 = Add(A=@n2, B=CameraVectorWS())
  LinearInterpolate(
   A=DotProduct(A=ReflectionVectorWS(), B=@n2) {desc=Phong}
   B=DotProduct(A=Normalize(VectorInput=@n3), B=PixelNormalWS()) {desc=Blinn}
   Alpha=ScalarParameter() {defaultValue=1, group=1_RoseDust, parameterName=BlendPhongBlinn, sliderMax=1, sortPriority=20}
  )
  n4 = Constant() {r=1}
  n5 = Fresnel() {baseReflectFraction=0, exponent=1}
  LinearInterpolate(A=@n4, B=Multiply(A=Power(Base=@n4) {constExponent=1.5}) {constB=0.75}, Alpha=Round(None=Saturate(None=@n5)))
  n6 = Subtract(A=Distance(A=ActorPositionWS(), B=CameraPositionWS())) {constB=100}
  LinearInterpolate(B=Multiply() {constB=4}, Alpha=Saturate(None=Divide(A=@n6) {constB=500}))
  n7 = TextureSample() {samplerType=SAMPLERTYPE_Masks, texture=/Game/Characters/Player/Female/Medium/Bodies/F_MED_RoseDust/Textures/T_F_MED_RoseDust_Head_FX}
  n8 = FunctionInput(Preview=@n7.RGB) {inputName=FX, inputType=FunctionInput_Vector4, sortPriority=1}
  n9 = ChannelMaskParameter(None=@n8) {group=1_RoseDust, maskChannel=Green, parameterName=SuitTransitionMaskChannel, sortPriority=3, defaultValue=<rgba 0 1 0 0>}
  NamedRerouteDeclaration(None=@n9.R) {name=SuitTransitionMask, nodeColor=<rgba 0.5234 0 1 1>}
  n10 = ChannelMaskParameter(None=@n8) {group=1_RoseDust, parameterName=SuitMasksChannel, sortPriority=3}
  n11 = Multiply(A=@n10.R) {constB=3}
  n12 = Add(A=@n11) {constB=-1}
  NamedRerouteDeclaration(None=Saturate(None=Multiply(A=@n12) {constB=0.5}) {desc=0.333-1}) {name=SuitWhitePinkMask, nodeColor=<rgba 1 1 1 1>}
  n13 = VertexNormalWS()
  n14 = SmoothStep(
   Min=Constant() {r=0.2}
   Max=Constant() {r=0.22}
   Value=Multiply(A=Fresnel(Normal=@n13) {baseReflectFraction=0, exponent=1}, B=NamedRerouteUsage())
  )
  n15 = VectorParameter() {group=1_RoseDust, parameterName=SuitTransitionColorLeft, sortPriority=10, defaultValue=<rgba 0.1506 0 0.9375 1>}
  n16 = VectorParameter() {group=1_RoseDust, parameterName=SuitTransitionColorRight, sortPriority=10, defaultValue=<rgba 0 0.7135 0.093 1>}
  n17 = VertexInterpolator(VS=PreSkinnedPosition())
  n18 = Ceil(None=Saturate(None=ComponentMask(None=@n17.PS) {r=true}))
  NamedRerouteDeclaration(None=Multiply(
    A=NamedRerouteDeclaration(None=@n14) {name=TransitionMask, nodeColor=<rgba 0.3558 1 0.8493 1>}
    B=LinearInterpolate(A=@n15.RGB, B=@n16.RGB, Alpha=@n18)
   )) {name=Transition, nodeColor=<rgba 0 0.6667 0.059 1>}
  n19 = MaterialFunctionCall!(Vector=VectorParameter() {group=1_RoseDust, parameterName=RimLighVector, sortPriority=10, defaultValue=<rgba -1 0 1 1>}.RGB) {materialFunction=/Game/Packages/Fortress_SharedMaterials/Anime_Master_Material/MF_ProjectToScreenSpace}
  n20 = PixelNormalWS()
  n21 = Subtract(A=NamedRerouteDeclaration(None=DotProduct(A=@n19.Result, B=@n20)) {name=RimGradient, nodeColor=<rgba 0 0.3906 1 1>}, B=NamedRerouteUsage())
  n22 = Add(A=@n21) {constB=-0.25}
  n23 = DotProduct(A=VertexInterpolator(VS=PreSkinnedNormal()).PS, B=Constant3Vector() {constant=<rgba 0 1 0 0>})
  n24 = Multiply(A=@n23, B=@n23)
  n25 = NamedRerouteDeclaration(None=Multiply(A=@n24, B=@n24) {desc=reduce rim light on the Y axis}) {name=YMask, nodeColor=<rgba 0 1 0.5078 1>}
  n26 = NamedRerouteDeclaration(None=SmoothStep(Min=Constant(), Max=Constant() {r=0.05}, Value=LinearInterpolate(A=@n21, B=@n22, Alpha=@n25) {constA=1, constB=16})) {name=AccentRimMask, nodeColor=<rgba 0.6631 0.3995 0.9131 1>}
  n27 = VectorParameter() {group=1_RoseDust, parameterName=RimColor, sortPriority=10, defaultValue=<rgba 0.4 0 0.8 0.8>}
  NamedRerouteDeclaration(None=AppendVector(A=Multiply(A=@n26, B=@n27.RGB), B=@n27.A)) {name=AccentRim, nodeColor=<rgba 0.5132 0 1 1>}
  n28 = DotProduct(
   A=MaterialFunctionCall(Vector=VectorParameter() {group=1_RoseDust, parameterName=RimDotsVector, sortPriority=10, defaultValue=<rgba -1 0 0.5 1>}.RGB) {materialFunction=&MF_ProjectToScreenSpace}.Result
   B=PixelNormalWS()
  )
  n29 = ScalarParameter() {defaultValue=-0.4, group=1_RoseDust, parameterName=RimDotsOffset, sliderMax=0.5, sliderMin=-0.5, sortPriority=20}
  n30 = Subtract(
   A=Multiply(
    A=Add(A=@n28, B=@n29) {constB=-0.25}
    B=ScalarParameter() {defaultValue=1.5, group=1_RoseDust, parameterName=RimDotsTransition, sliderMax=2, sortPriority=20}
   ) {constB=1.5}
   B=NamedRerouteUsage()
  )
  n31 = Subtract(A=LinearInterpolate(A=@n30, B=Add(A=@n30) {constB=-0.25}, Alpha=NamedRerouteUsage()) {constA=1, constB=16}, B=NamedRerouteUsage())
  n32 = NamedRerouteDeclaration(None=SmoothStep(Min=Constant(), Max=Constant() {r=0.1}, Value=Saturate(None=@n31))) {name=RimDotsMask, nodeColor=<rgba 0.4766 1 0 1>}
  n33 = VectorParameter() {group=1_RoseDust, parameterName=RimDotsColor, sortPriority=10, defaultValue=<rgba 0.9 0.9 0.9 1>}
  NamedRerouteDeclaration(None=AppendVector(A=Multiply(A=@n32, B=@n33.RGB), B=@n33.A)) {name=RimDots, nodeColor=<rgba 1 0 0.6797 1>}
  n34 = Multiply(A=NamedRerouteUsage()) {constB=-1}
  n35 = ScalarParameter() {group=1_RoseDust, parameterName=ShadowHatchingOffset, sliderMax=0.5, sliderMin=-0.5, sortPriority=20}
  n36 = Subtract(A=Multiply(A=Add(A=@n34, B=@n35) {constB=-0.25}) {constB=0.5, desc=:/}, B=NamedRerouteUsage())
  n37 = NamedRerouteDeclaration(None=SmoothStep(Min=Constant(), Max=Constant() {r=0.1}, Value=Saturate(None=@n36))) {name=ShadowHatchingMask, nodeColor=<rgba 0.3594 0.3594 0.3594 1>}
  n38 = VectorParameter() {group=1_RoseDust, parameterName=ShadowHatchingColor, sortPriority=10, defaultValue=<rgba 0.18 0.18 0.18 0.5>}
  NamedRerouteDeclaration(None=AppendVector(A=Multiply(A=@n37, B=@n38.RGB), B=@n38.A)) {name=ShadowHatching, nodeColor=<rgba 0.099 0.099 0.099 1>}
  NamedRerouteDeclaration(None=VertexColor().G) {name=SuitMask, nodeColor=<rgba 0.3125 0 1 1>}
  n39 = Add(A=OneMinus(None=@n10.R)) {constB=-0.666}
  NamedRerouteDeclaration(None=Multiply(A=Saturate(None=@n11) {desc=0-0.333}, B=Ceil(None=@n39))) {name=SuitBlackMask, nodeColor=<rgba 0.0156 0.0156 0.0156 1>}
  n40 = MaterialFunctionCall() {materialFunction=&CameraDirectionVector}
  n41 = Multiply(A=Constant3Vector() {constant=<rgba 1 1 0 0>}, B=@n40.Result)
  n42 = Add(A=Multiply(B=Normalize(VectorInput=@n41)) {constA=0.25}, B=@n40.Result)
  n43 = Multiply(A=@n40.Result) {constB=50}
  n44 = Add(A=ActorPositionWS(), B=Constant3Vector() {constant=<rgba 0 0 90 1>})
  n45 = Constant() {r=64}
  n46 = Constant() {r=128}
  n47 = InverseLinearInterpolate(
   A=Constant() {r=45}
   B=Constant() {r=90}
   Value=MaterialFunctionCall() {materialFunction=/Engine/Functions/Engine_MaterialFunctions02/Utility/FOV}.FOV angle in degrees
  )
  NamedRerouteDeclaration(None=MaterialFunctionCall(Plane Normal Axis=Normalize(VectorInput=@n42), Plane Center=Add(A=@n43, B=@n44), UV Scale=LinearInterpolate(A=@n45, B=@n46, Alpha=@n47)) {materialFunction=&VirtualPlaneCoordinates}.UVs) {name=ObjectAlighedScreenPosition, nodeColor=<rgba 1 0.867 0 1>}
  n48 = ChannelMaskParameter(None=@n8) {group=1_RoseDust, maskChannel=Blue, parameterName=SkinHairShadingMaskChannel, sortPriority=3, defaultValue=<rgba 0 0 1 0>}
  NamedRerouteDeclaration(None=@n48.R) {name=SkinHairShadingMask, nodeColor=<rgba 0.8984 1 0 1>}
  n49 = Custom() {code="return View.DirectionalLightDirection;", description=DirectionalLightDirection, inputs []}
  n50 = PixelNormalWS()
  n51 = Multiply(A=NamedRerouteDeclaration(None=DotProduct(A=@n49, B=@n50)) {name=LdotN, nodeColor=<rgba 0.2656 0 1 1>}) {constB=0.5}
  NamedRerouteDeclaration(None=Saturate(None=Add(A=@n51) {constB=0.5})) {name=Lambert, nodeColor=<rgba 0 0.9297 1 1>}
  n52 = Subtract(A=Distance(A=ActorPositionWS(), B=CameraPositionWS())) {constB=100}
  NamedRerouteDeclaration(None=Saturate(None=Divide(A=@n52) {constB=400})) {name=DistanceBlending, nodeColor=<rgba 0.4766 1 0 1>}
  n53 = ScalarParameter() {defaultValue=150, group=1_RoseDust, parameterName=HatchingAmount, sliderMax=512, sortPriority=20}
  n54 = Multiply(A=NamedRerouteUsage(), B=LinearInterpolate(A=@n53, B=Multiply(A=@n53) {constB=0.5}, Alpha=NamedRerouteUsage()))
  n55 = Constant() {r=1}
  n56 = Frac(None=ComponentMask(None=Rotator(Coordinate=@n54, Time=@n55) {speed=0.75}) {g=true})
  NamedRerouteDeclaration(None=Abs(None=Add(A=@n56) {constB=-0.5})) {name=HatchingGradient, nodeColor=<rgba 0.3042 0 1 1>}
  n57 = VectorParameter() {group=1_RoseDust, parameterName=RimHatchingVector, sortPriority=10, defaultValue=<rgba 0.8229 0.15 0.4865 1>}
  n58 = DotProduct(
   A=MaterialFunctionCall(Vector=@n57.RGB) {materialFunction=&MF_ProjectToScreenSpace}.Result
   B=PixelNormalWS()
  )
  n59 = ScalarParameter() {defaultValue=-0.3, group=1_RoseDust, parameterName=HatchingOffset, sliderMax=0.5, sliderMin=-0.5, sortPriority=20}
  n60 = Subtract(A=Multiply(A=Add(A=@n58, B=@n59) {constB=-0.25}) {constB=0.5, desc=:/}, B=NamedRerouteUsage())
  n61 = NamedRerouteDeclaration(None=SmoothStep(Min=Constant(), Max=Constant() {r=0.1}, Value=Saturate(None=@n60))) {name=RimHatchingMask, nodeColor=<rgba 0.6174 0.2284 0.6354 1>}
  n62 = VectorParameter() {group=1_RoseDust, parameterName=RimHatchingColor, sortPriority=10, defaultValue=<rgba 0.4 0 0.8 0.8>}
  NamedRerouteDeclaration(None=AppendVector(A=Multiply(A=@n61, B=@n62.RGB), B=@n62.A)) {name=RimHatching, nodeColor=<rgba 0.651 0 0.589 1>}
  n63 = GetMaterialAttributes(None=NamedRerouteUsage()) {attributeGetTypes [#1, #2]}
  n64 = GetMaterialAttributes(None=NamedRerouteUsage()) {attributeGetTypes [#2]}
  n65 = NamedRerouteUsage()
  n66 = Add(
   A=Multiply(A=Multiply(A=@n64.BaseColor, B=@n65), B=ScalarParameter() {defaultValue=0.15, group=1_RoseDust, parameterName=SuitEmissive, sliderMax=1}) {constB=0.5}
   B=NamedRerouteUsage()
  )
  n67 = NamedRerouteUsage()
  n68 = NamedRerouteUsage()
  n69 = LinearInterpolate(A=NamedRerouteUsage(), B=NamedRerouteUsage(), Alpha=NamedRerouteUsage())
  n70 = ComponentMask(None=@n69) {a=true}
  n71 = Multiply(A=Max(A=NamedRerouteUsage(), B=NamedRerouteUsage()), B=NamedRerouteUsage())
  n72 = NamedRerouteUsage()
  n73 = ComponentMask(None=@n72) {a=true}
  n74 = Multiply(A=NamedRerouteUsage(), B=NamedRerouteUsage())
  n75 = Add(
   A=Multiply(A=ComponentMask(None=@n69) {b=true, g=true, r=true}, B=Multiply(A=@n70, B=@n71)) {desc=Left RIm}
   B=Multiply(A=ComponentMask(None=@n72) {b=true, g=true, r=true}, B=Multiply(A=@n73, B=@n74)) {desc=Right RIm}
  )
  n76 = Saturate(None=Add(A=LinearInterpolate(A=@n66, B=@n67, Alpha=@n68) {desc=blend in the transition stripes}, B=@n75))
  n77 = Multiply(A=Max(A=NamedRerouteUsage(), B=NamedRerouteUsage()), B=Max(A=NamedRerouteUsage(), B=NamedRerouteUsage()))
  n78 = NamedRerouteUsage()
  n79 = NamedRerouteUsage()
  n80 = NamedRerouteUsage()
  n81 = Saturate(None=Add(A=Multiply(A=@n77, B=@n78), B=Multiply(A=@n79, B=@n80)) {constB=0})
  n82 = SetMaterialAttributes!(
   MaterialAttributes=@n63.MaterialAttributes
   EmissiveColor=Add(
    A=@n63.EmissiveColor
    B=MaterialFunctionCall!(In=@n76, LUT, LUT Multiplier) {materialFunction=/Game/Packages/Fortress_SharedMaterials/Anime_Master_Material/MF_ApplyLUT}.Result
   )
   BaseColor=LinearInterpolate(A=LinearInterpolate(A=@n63.BaseColor, B=@n75, Alpha=@n81), B=NamedRerouteUsage(), Alpha=NamedRerouteUsage())
  ) {desc=screen projected rim and hatching, attributeSetTypes [#1, #2]}
  n83 = GetMaterialAttributes(None=NamedRerouteDeclaration(None=@n82) {name=AccentRimMaterial, nodeColor=<rgba 0.3359 0 1 1>}) {attributeGetTypes [#1, #2]}
  n84 = NamedRerouteUsage()
  n85 = Multiply(A=Multiply(A=ComponentMask(None=@n84) {a=true}, B=NamedRerouteUsage()), B=NamedRerouteUsage())
  n86 = Multiply(A=@n83.BaseColor, B=ComponentMask(None=@n84) {b=true, g=true, r=true})
  NamedRerouteDeclaration(None=SetMaterialAttributes(
    MaterialAttributes=@n83.MaterialAttributes
    EmissiveColor=LinearInterpolate(A=@n83.EmissiveColor, Alpha=@n85) {constB=0}
    BaseColor=LinearInterpolate(A=@n83.BaseColor, B=@n86, Alpha=@n85)
   ) {attributeSetTypes [#1, #2]}) {name=ShadowHatchingMaterial, nodeColor=<rgba 0 1 0.1328 1>}
  n87 = ScalarParameter() {defaultValue=150, group=1_RoseDust, parameterName=DotsAmount, sliderMax=512, sortPriority=20}
  n88 = Multiply(A=NamedRerouteUsage(), B=LinearInterpolate(A=@n87, B=Multiply(A=@n87) {constB=0.5}, Alpha=NamedRerouteUsage()))
  n89 = Constant() {r=1}
  n90 = Add(A=Frac(None=Rotator(Coordinate=@n88, Time=@n89) {speed=0.75})) {constB=-0.5}
  n91 = Constant()
  NamedRerouteDeclaration(None=Add(A=Distance(A=@n90, B=@n91)) {constB=-0.1, desc=to get full transition}) {name=DotsGradient, nodeColor=<rgba 1 0.0938 0 1>}
  n92 = NamedRerouteUsage()
  n93 = ScalarParameter() {defaultValue=-0.3, group=1_RoseDust, parameterName=HighlightDotsOffset, sliderMax=0.5, sliderMin=-0.5, sortPriority=20}
  n94 = Subtract(
   A=Multiply(
    A=Add(A=@n92, B=@n93) {constB=-0.25}
    B=ScalarParameter() {defaultValue=1.5, group=1_RoseDust, parameterName=HighlightDotsTransition, sliderMax=2, sortPriority=20}
   ) {constB=1.5}
   B=NamedRerouteUsage()
  )
  n95 = NamedRerouteDeclaration(None=SmoothStep(Min=Constant(), Max=Constant() {r=0.1}, Value=Saturate(None=@n94))) {name=HighlightDotsMask, nodeColor=<rgba 0.4766 1 0 1>}
  n96 = VectorParameter() {group=1_RoseDust, parameterName=HighlightDotsColor, sortPriority=10, defaultValue=<rgba 0.9 0.9 0.9 1>}
  NamedRerouteDeclaration(None=AppendVector(A=Multiply(A=@n95, B=@n96.RGB), B=@n96.A)) {name=HighlightDots, nodeColor=<rgba 1 0 0.6797 1>}
  n97 = ChannelMaskParameter(None=@n8) {group=1_RoseDust, maskChannel=Green, parameterName=ThresholdChannel, sortPriority=3, defaultValue=<rgba 0 1 0 0>}
  NamedRerouteDeclaration(None=@n97.R) {name=ThresholdMap, nodeColor=<rgba 0.7109 1 0 1>}
  n98 = VectorParameter() {group=1_RoseDust, parameterName=WhiteSuitShadowHatchingColor, sortPriority=10, defaultValue=<rgba 0.18 0.18 0.18 0.5>}
  NamedRerouteDeclaration(None=AppendVector(A=Multiply(A=@n37, B=@n98.RGB), B=@n98.A)) {name=WhiteSuitShadowHatching, nodeColor=<rgba 0.8594 0.8594 0.8594 1>}
  n99 = TextureSample() {samplerType=SAMPLERTYPE_Masks, texture=/Game/Characters/Player/Female/Medium/Bodies/F_MED_RoseDust/Textures/T_F_MED_RoseDust_Body_S}
  n100 = TextureSample() {samplerType=SAMPLERTYPE_Normal, texture=/Game/Characters/Player/Female/Medium/Bodies/F_MED_RoseDust/Textures/T_F_MED_RoseDust_Body_N}
  n101 = SetMaterialAttributes!(
   MaterialAttributes
   BaseColor=TextureSample() {texture=/Game/Characters/Player/Female/Medium/Bodies/F_MED_RoseDust/Textures/T_F_MED_RoseDust_Body_D}.RGB
   Specular=@n99.R
   Metallic=@n99.R
   Roughness=@n99.R
   Normal=@n100.RGB
  ) {attributeSetTypes [
    #2
    #3
    #4
    #5
    #6
   ]}
  n102 = GetMaterialAttributes(None=FunctionInput(Preview=@n101) {inputName=MA, inputType=FunctionInput_MaterialAttributes}) {attributeGetTypes [#4, #5]}
  n103 = Constant()
  NamedRerouteDeclaration(None=StaticSwitchParameter(True=@n102.Metallic, False=@n103) {defaultValue=true, group=1_RoseDust, parameterName=UseMetallicAsSheen, sortPriority=1}) {name=Sheen, nodeColor=<rgba 0.217 0.2569 1 1>}
  n104 = StaticSwitchParameter(True=@n103, False=@n102.Metallic) {defaultValue=true, group=1_RoseDust, parameterName=UseMetallicAsSheen, sortPriority=1}
  NamedRerouteDeclaration(None=SetMaterialAttributes!(MaterialAttributes=@n102.MaterialAttributes, Metallic=NamedRerouteDeclaration(None=@n104) {name=MetallicMask, nodeColor=<rgba 0 1 0.2266 1>}) {attributeSetTypes [#4]}) {name=MA, nodeColor=<rgba 0.3594 1 0 1>}
  n105 = GetMaterialAttributes(None=NamedRerouteUsage()) {attributeGetTypes [#2]}
  n106 = NamedRerouteUsage()
  n107 = NamedRerouteUsage()
  n108 = NamedRerouteUsage()
  n109 = ComponentMask(None=@n108) {b=true, g=true, r=true}
  n110 = ComponentMask(None=@n108) {a=true}
  n111 = Add(
   A=Multiply(A=ComponentMask(None=@n107) {b=true, g=true, r=true}, B=ComponentMask(None=@n107) {a=true})
   B=Subtract(A=NamedRerouteUsage(), B=Multiply(A=@n109, B=@n110))
  )
  n112 = ScalarParameter() {group=1_RoseDust, parameterName=MinShadingValue, sliderMax=1}
  n113 = ScalarParameter() {defaultValue=1, group=1_RoseDust, parameterName=MaxShadingValue, sliderMax=2}
  n114 = Multiply(A=Multiply(A=@n105.BaseColor, B=Clamp(None=@n111, Min=@n112, Max=@n113))) {constB=0.5}
  n115 = SetMaterialAttributes!(
   MaterialAttributes=@n105.MaterialAttributes
   BaseColor=LinearInterpolate(A=Multiply(A=@n105.BaseColor) {constB=0.5}, B=@n105.BaseColor, Alpha=@n106)
   EmissiveColor=LinearInterpolate(
    A=MaterialFunctionCall(In=@n114) {materialFunction=&MF_ApplyLUT}.Result
    Alpha=@n106
   ) {constB=0}
  ) {attributeSetTypes [#2, #1]}
  n116 = NamedRerouteUsage()
  n117 = NamedRerouteUsage()
  n118 = NamedRerouteUsage()
  n119 = NamedRerouteUsage()
  n120 = ScalarParameter() {defaultValue=1, group=1_RoseDust, parameterName=SuitMetalness, sliderMax=1}
  n121 = SetMaterialAttributes!(
   MaterialAttributes=StaticSwitchParameter(True=BlendMaterialAttributes(A=@n118, B=@n116, Alpha=@n117), False=@n118) {group=1_RoseDust, parameterName=UseSkinHairShading_FXMask, sortPriority=1}
   Metallic=Saturate(None=Multiply(A=@n119, B=@n120))
   Roughness=Multiply(
    A=NamedRerouteUsage()
    B=ScalarParameter() {defaultValue=2, group=1_RoseDust, parameterName=SuitRoughness_Scale, sliderMax=4, sortPriority=30}
   )
  ) {attributeSetTypes [#4, #5]}
  n122 = NamedRerouteUsage()
  NamedRerouteDeclaration(None=StaticSwitchParameter(
    True=NamedRerouteDeclaration(None=@n115) {name=SkinHairShadingMaterial, nodeColor=<rgba 1 0 0.5391 1>}
    False=StaticSwitchParameter(True=@n121, False=@n122) {group=1_RoseDust, parameterName=UseSuitShading, sortPriority=1}
   ) {group=1_RoseDust, parameterName=UseSkinHairShading, sortPriority=1}) {name=MA_Out, nodeColor=<rgba 1 0.0703 0 1>}
  n123 = CameraVectorWS()
  n124 = PixelNormalWS()
  n125 = Power(
   Base=Saturate(None=DotProduct(A=@n123, B=@n124))
   Exp=ScalarParameter() {defaultValue=20, group=1_RoseDust, parameterName=SpecPower, sliderMax=32, sortPriority=20}
  )
  n126 = ScalarParameter() {defaultValue=0.1, group=1_RoseDust, parameterName=SpecularIntensity, sliderMax=2, sortPriority=20}
  NamedRerouteDeclaration(None=Multiply(A=NamedRerouteUsage(), B=Multiply(A=@n125, B=@n126))) {name=Specular, nodeColor=<rgba 0 1 1 1>}
  NamedRerouteDeclaration(None=@n102.Roughness) {name=Roughness_In, nodeColor=<rgba 1 0.4453 0 1>}
  NamedRerouteUsage()
  n127 = Constant3Vector() {constant=<rgba 1 1 0 0>}
  n128 = MaterialFunctionCall() {materialFunction=&CameraDirectionVector}
  n129 = Multiply(
   A=Constant() {desc=Needs at least a small amount of this or it will break, r=0.25}
   B=Normalize(VectorInput=Multiply(A=@n127, B=@n128.Result))
  )
  n130 = Multiply(A=@n128.Result, B=ScalarParameter() {defaultValue=70, group=Cartoon, parameterName=CRTN_VP-CameraVectorMin})
  n131 = StaticSwitchParameter(
   True=ObjectPositionWS()
   False=Add(A=ActorPositionWS(), B=VectorParameter() {group=Cartoon, parameterName=CRTN_VP-ActorPosOffset, defaultValue=<rgba 0 0 80 1>}.RGB)
  ) {group=Cartoon, parameterName=CRTN_UseObjectPos}
  n132 = MaterialFunctionCall() {materialFunction=&FOV}
  n133 = Constant() {r=0.414}
  n134 = Constant() {r=1.091}
  n135 = Constant()
  n136 = Constant() {r=1}
  n137 = LinearInterpolate(
   A=ScalarParameter() {defaultValue=1.4, group=Cartoon, parameterName=CRTN_VP-Scale-LowFOV}
   B=ScalarParameter() {defaultValue=1.75, group=Cartoon, parameterName=CRTN_VP-Scale-HighFOV}
   Alpha=Saturate(None=MaterialFunctionCall(Input=@n132."FOV ", Input Low=@n133, Input High=@n134, Target Low=@n135, Target High=@n136) {materialFunction=&RemapValueRange}.Result)
  )
  TextureSample(UVs=MaterialFunctionCall(Plane Normal Axis=Add(A=@n129, B=@n128.Result), Plane Center=Add(A=@n130, B=@n131), UV Scale=Multiply(A=@n137) {constB=128}) {materialFunction=&VirtualPlaneCoordinates}.UVs) {texture=/Game/FN_PublicTurntable/Turntable_Lowpoly/Content/BlueprintAssets/Checker}
  TextureSample() {samplerType=SAMPLERTYPE_Masks, texture=/Game/Characters/Player/Female/Medium/Bodies/F_MED_RoseDust/Textures/T_F_MED_RoseDust_Mask_FX}
  TextureSample() {samplerType=SAMPLERTYPE_Masks, texture=/Game/Characters/Player/Female/Medium/Bodies/F_MED_RoseDust/Textures/T_F_MED_RoseDust_Body_FX}
  TextureSample() {texture=&Checker}
 ]}
