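# Material function for the RoseDust character body that adds an optional thin-film iridescent sheen to skin material attributes, gated by the UseThinFilm switch.
# A spectrum texture is sampled with warped Fresnel UVs, shifted by a static offset or optional modulo-time animation.
# Its color lerps into base color and emissive, weighted by inverted roughness and a second Fresnel.
# A selectable SkinFX channel masks the blend back into the incoming attributes.
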MaterialFunction /Game/Characters/Player/Female/Medium/Bodies/F_MED_RoseDust/Materials/MF_RoseDust_ThinFilm
graph {nodes [
  FunctionOutput(None=StaticSwitchParameter(True=NamedRerouteUsage(), False=NamedRerouteUsage()) {group=Skin FX, parameterName=UseThinFilm})
  NamedRerouteDeclaration(None=FunctionInput() {inputName=Material Attributes, inputType=FunctionInput_MaterialAttributes}) {name=MA, nodeColor=<rgba 0.6641 0 1 1>}
  NamedRerouteDeclaration(None=StaticComponentMaskParameter(None=FunctionInput() {inputName=SkinFX, inputType=FunctionInput_Vector4, sortPriority=1}) {defaultR=true, group=ThinFilm, parameterName=ThinFilm_Channel, sortPriority=0}) {name=AlphaMask, nodeColor=<rgba 0.0078 1 0 1>}
  n1 = GetMaterialAttributes(None=NamedRerouteUsage()) {attributeGetTypes [#1, #2, #3]}
  n2 = MaterialFunctionCall!(
   Time
   Divisor=ScalarParameter() {defaultValue=1, group=ThinFilm, parameterName=ThinFilm_Animated_TimeDivisor, sliderMax=1, sortPriority=10}
   Mod=ScalarParameter() {defaultValue=1, group=ThinFilm, parameterName=ThinFilm_Animated_TimeModulo, sliderMax=1, sortPriority=10}
  ) {materialFunction=/Game/Effects/Fort_Effects/Materials/_Functions/MF_ModuloTime}
  n3 = ScalarParameter() {group=ThinFilm, parameterName=ThinFilm_Offset, sliderMax=1, sortPriority=11}
  n4 = ScalarParameter() {defaultValue=1, group=ThinFilm, parameterName=ThinFilm_Scale, sliderMax=1, sortPriority=11}
  n5 = Fresnel(ExponentIn=ScalarParameter() {defaultValue=0.5, group=ThinFilm, parameterName=ThinFilm_Warp, sliderMax=1, sortPriority=11}) {baseReflectFraction=0}
  n6 = TextureSampleParameter2D(UVs=Add(
    A=StaticSwitchParameter(True=@n2.Unclamped, False=@n3) {group=ThinFilm, parameterName=AnimateThinFilm, sortPriority=1}
    B=Multiply(A=@n4, B=@n5)
   )) {group=ThinFilm, parameterName=Thin Film Texture, sortPriority=2, texture=/Game/Global/Textures/Default/Thinfilm/T_ThinFilm_Spectrum_COLOR}
  n7 = OneMinus(None=@n1.Roughness)
  n8 = ScalarParameter() {defaultValue=1, group=ThinFilm, parameterName=ThinFilm_RoughnessScale, sliderMax=1, sortPriority=5}
  n9 = Multiply(
   A=Saturate(None=Multiply(A=@n7, B=@n8))
   B=Fresnel(ExponentIn=ScalarParameter() {defaultValue=1.25, group=ThinFilm, parameterName=ThinFilm_Exponent, sliderMax=4, sortPriority=5}) {baseReflectFraction=0}
  )
  n10 = LinearInterpolate(
   A=@n1.BaseColor
   B=Multiply(A=@n6.RGB, B=ScalarParameter() {defaultValue=0.5, group=ThinFilm, parameterName=ThinFilm_Intensity, sliderMax=1, sortPriority=3})
   Alpha=@n9
  )
  n11 = LinearInterpolate(
   A=@n1.EmissiveColor
   B=Multiply(A=@n6.RGB, B=ScalarParameter() {group=ThinFilm, parameterName=ThinFilm_Emissive_Intensity, sliderMax=1, sortPriority=4})
   Alpha=@n9
  )
  NamedRerouteDeclaration(None=BlendMaterialAttributes(
    A=NamedRerouteUsage()
    B=SetMaterialAttributes!(MaterialAttributes=@n1.MaterialAttributes, BaseColor=@n10, EmissiveColor=@n11) {attributeSetTypes [#2, #3]}
    Alpha=NamedRerouteUsage()
   ) {vertexAttributeBlendType=UseA}) {name=MA_Output, nodeColor=<rgba 0.3359 1 0 1>}
 ]}
