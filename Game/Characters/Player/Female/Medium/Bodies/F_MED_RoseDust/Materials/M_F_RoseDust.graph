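# Master character material for the RoseDust female player body, using material attributes and a shading model chosen in the graph.
# Between the shared character PreFX and PostFX functions, it chains RoseDust thin-film, distance-field inkline and highlight functions, then ends in a cinematic post-FX pass.
# A SkinFX mask texture is shared through named reroutes.
# A static switch optionally applies the subsurface shading model function, driven by selectable primary and secondary mask channels.
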Material /Game/Characters/Player/Female/Medium/Bodies/F_MED_RoseDust/Materials/M_F_RoseDust
graph {
 properties {
  bAutomaticallySetUsageInEditor=false
  bUseMaterialAttributes=true
  bUsedWithClothing=true
  bUsedWithMeshDeformer=true
  bUsedWithMorphTargets=true
  bUsedWithNanite=true
  bUsedWithSkeletalMesh=true
  pixelDepthOffsetMode=PDOM_Legacy
  shadingModel=MSM_FromMaterialExpression
  subsurfaceProfile=&SS_HeroSkin_02
  usedShadingModels=Default Lit | Subsurface | Subsurface Profile
 }
 nodes [
  n1 = MaterialFunctionCall() {materialFunction=&BaseFN_CharacterFunction_PreFX}
  n2 = NamedRerouteUsage()
  n3 = StaticComponentMaskParameter(None=@n2) {defaultG=true, group=Skin FX, parameterName=SS Masks: Dual SS Secondary Mask}
  n4 = StaticComponentMaskParameter(None=@n2) {defaultR=true, group=Skin FX, parameterName=SS Masks}
  n5 = NamedRerouteUsage()
  n6 = MaterialFunctionCall!(
   MA=MaterialFunctionCall!(DFMap=MaterialFunctionCall!(Material Attributes=@n1.Result, SkinFX=@n5) {materialFunction=&/MF_RoseDust_ThinFilm}.Result) {materialFunction=&/MF_RoseDust_DF_Inklines}.Result
   FX=NamedRerouteUsage()
  ) {materialFunction=&/MF_RoseDust_Highlights}
  n7 = MaterialFunctionCall(
   AO=@n1.AO
   Distance Blend=@n1.AO
   Fresnel=@n1.AO
   Gmap=@n1.AO
   Pre Skinned Local Position=@n1.AO
   Material=StaticSwitchParameter(
    True=MaterialFunctionCall!(In=@n6.Result, Dual SS Secondary Alpha=@n3, Alpha=@n4) {materialFunction=/Game/Weapons/WeaponSkins/MaterialFunction/SS_Shader_Model}.Result
    False=@n6.Result
   ) {group=Skin FX, parameterName="Uses Subsurface Shading Model (NOT Subsurface Profile)"}
  ) {materialFunction=&BaseFN_CharacterFunction_PostFX}
  MaterialFunctionCall(In=@n7.Result) {materialFunction=&CINE_MF_CharPostFX}
  n8 = TextureSampleParameter2D() {
   group=Skin FX
   parameterName=SkinFX_Mask
   samplerType=SAMPLERTYPE_Masks
   sortPriority=1
   texture=&T_BlackMask
  }
  NamedRerouteDeclaration(None=@n8.RGBA) {name=SkinFX, nodeColor=<rgba 0 1 0.8359 1>}
 ]
}
